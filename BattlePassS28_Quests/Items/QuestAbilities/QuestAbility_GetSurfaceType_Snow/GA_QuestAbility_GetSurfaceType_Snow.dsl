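; Quest gameplay ability that reports while the player is standing on snow landscape.
; On Event.SurfaceType.Changed, after a short retriggerable delay, it line-traces down from the instigator and sets OnSnow if it hits Landscape with the Snow physical material.
; A looping one-second timer, cleared on ability end, broadcasts CustomVerb.SurfaceCheck.Snow for the avatar while OnSnow is true.
; Debug prints show the surface and tick state.
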(event Ability|EventActivateAbility
  (bind _returnvalue (Utilities|Time|SetTimerbyEvent (delegate Custom|SurfaceCheckTick) 1.0 true))
  (Variables|Default|SetValidSurfaceTimer _returnvalue)
  ;[0.25 1 0.25] Is Player on Snow?
  (bind (_asynctaskproxy _payload) (Ability|Async|WaitGameplayEventToActor (Ability|GetAvatarActorfromActorInfo) "(TagName=\"Event.SurfaceType.Changed\")")
    (:EventReceived
      (Utilities|FlowControl|RetriggerableDelay 0.1)
      (bind _outhit (Collision|LineTraceForObjects (Transformation|GetActorLocation (.Instigator (Utilities|Struct|BreakGameplayEventData _payload))) (+ (Transformation|GetActorLocation (.Instigator (Utilities|Struct|BreakGameplayEventData _payload))) (* (Transformation|GetActorUpVector (.Instigator (Utilities|Struct|BreakGameplayEventData _payload))) -5e+02)) (Utilities|Array|MakeArray "ObjectTypeQuery1") true (Utilities|Array|MakeArray (.Instigator (Utilities|Struct|BreakGameplayEventData _payload))) "None" true "(R=1.000000,G=0.000000,B=0.000000,A=1.000000)" "(R=0.000000,G=1.000000,B=0.000000,A=1.000000)" 1.0))
      (if (and (== (.PhysMat (Collision|BreakHitResult _outhit)) "/Game/FortressPhysicalMaterials/Snow.Snow") (== (Utilities|GetClass (.HitActor (Collision|BreakHitResult _outhit))) "/Script/Landscape.Landscape"))
        (Variables|Default|SetOnSnow true)
        (Development|PrintString "Snow!" false false "(R=0.083333,G=1.000000,B=0.000000,A=1.000000)")
        (else
          (Variables|Default|SetOnSnow false)))
      (Development|PrintString (Utilities|GetDisplayName (.PhysMat (Collision|BreakHitResult _outhit))) false false))))

(event Custom|SurfaceCheckTick
  (if (Variables|Default|GetOnSnow)
    (CustomVerb|BroadcastCustomVerbMessage (CustomVerb|MakeCustomVerbMessage (Ability|GetAvatarActorfromActorInfo) "(TagName=\"CustomVerb.SurfaceCheck.Snow\")"))
    (Development|PrintString "Tick Snow" false false)
    (else
      (Development|PrintString "Tick NO Snow" false false))))

(event Ability|EventOnEndAbility
  (params (bWasCancelled :type Boolean))
  (Utilities|Time|ClearandInvalidateTimerbyHandle (Variables|Default|GetValidSurfaceTimer)))
